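# Physics asset for the WaveTrail backpack cosmetic, providing collision bodies and joints for a dangling rope with two tags.
# Two capsules approximate the pack, small capsules form the rope chain, and thin boxes represent the tags.
# The rope base joins the pack with free swing and twist, and both tags branch from the last rope segment.
# Joints below the base have narrow swing limits and zero-degree twist limits.
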PhysicsAsset /BRCosmetics/Accessories/FORT_Backpacks/Backpack_F_MED_WaveTrail/Meshes/F_MED_WaveTrail_Pack_Physics
bodies {
 base_rope [{shapeType=Capsule, length=0.505, radius=0.505, shapeName=base_rope_capsule, rotation=<rot 0 -90 0>}]
 dyn_rope_01 [{shapeType=Capsule, length=1.1542, radius=0.7, shapeName=dyn_rope_01_capsule, center=<2 0 0>, rotation=<rot 90 0 0>}]
 dyn_rope_02 [{shapeType=Capsule, length=0.45, radius=1.2145, shapeName=dyn_rope_02_capsule, center=<2.5 0 0>, rotation=<rot 90 0 0>}]
 dyn_rope_03 [{shapeType=Capsule, length=1.2708, radius=0.5, shapeName=dyn_rope_03_capsule, center=<2 0 0>, rotation=<rot 90 0 0>}]
 dyn_tag_01 [{shapeType=Box, extentX=0.1, extentY=2.5, extentZ=3.5, shapeName=dyn_tag_01_box, center=<2 0 0>, rotation=<rot 0 -90 -90>}]
 dyn_tag_02 [{shapeType=Box, extentX=0.1, extentY=3, extentZ=3.5, shapeName=dyn_tag_02_box, center=<2 0 0.4092>, rotation=<rot 0 -90 -90>}]
 pack [
  {
   shapeType=Capsule
   length=10.0742
   radius=8.7032
   shapeName=pack_capsule
   center=<4.7049 11.3139 -19.9834>
   rotation=<rot -6.367 -92.456 15.9315>
  }
  {shapeType=Capsule, length=15, radius=7.5, shapeName=pack_capsule, center=<-3.8979 -5.2163 -11.1692>, rotation=<rot -23.7596 0.193 -4.7899>}
 ]
}
constraints [
 {bone1Name=dyn_rope_01, bone2Name=base_rope, swing1LimitDegrees=20, swing2LimitDegrees=18.5, twistLimitDegrees=0}
 {bone1Name=dyn_rope_02, bone2Name=dyn_rope_01, swing1LimitDegrees=15, swing2LimitDegrees=18.5, twistLimitDegrees=0}
 {bone1Name=dyn_rope_03, bone2Name=dyn_rope_02, swing1LimitDegrees=15, swing2LimitDegrees=18.5, twistLimitDegrees=0}
 {
  bone1Name=base_rope
  bone2Name=pack
  swing1LimitDegrees=45
  swing1Motion=Free
  swing2LimitDegrees=45
  swing2Motion=Free
  twistLimitDegrees=45
  twistMotion=Free
 }
 {bone1Name=dyn_tag_01, bone2Name=dyn_rope_03, swing1LimitDegrees=10, swing2LimitDegrees=18.5, twistLimitDegrees=0}
 {bone1Name=dyn_tag_02, bone2Name=dyn_rope_03, swing1LimitDegrees=5, swing2LimitDegrees=5, twistLimitDegrees=0}
]
